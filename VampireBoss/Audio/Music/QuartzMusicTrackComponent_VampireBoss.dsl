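; Blueprint for the Vampire Boss music track component, apparently built on a Quartz-based music system.
; It overrides IsEnabled to always return true, so the track always reports itself as enabled.
; No other logic is defined in this file.
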(fn IsEnabled
  (return true))
