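# Sound cue that plays the Vampire Boss death music stinger, named as the boss elimination stinger.
# A single wave player feeds the Mx_VampireBoss_Death_Stinger wave straight to the root, with no mixing, randomization or modulation.
# It routes through the NPC music sound class with boosted volume, a very large max distance and play-when-silent virtualization.
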SoundCue /VampireBoss/Audio/Music/VampireBoss_Elim_Stinger_Cue
root=WavePlayer_0
nodes [{id=WavePlayer_0, properties {soundWaveAssetPtr=&/Mx_VampireBoss_Death_Stinger}}]
properties {
 maxDistance=2097152
 soundClassObject=&Fort_Music_NPC
 virtualizationMode=PlayWhenSilent
 volumeMultiplier=1.5
}
