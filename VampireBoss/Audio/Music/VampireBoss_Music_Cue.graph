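# Sound cue for Vampire Boss music that shifts between idle and aggro loops based on awareness parameters.
# A mixer sums three continuous-modulator branches whose volumes follow the Unaware, Alerted and Threaten parameters, each defaulting to silent.
# Unaware feeds the looping idle track, while Alerted and Threaten share one looping aggro wave player.
# Output uses the NPC music sound class and sends to the music audio-analysis submix.
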SoundCue /VampireBoss/Audio/Music/VampireBoss_Music_Cue
root=Mixer_1
nodes [
 {id=Mixer_1, children [ModulatorContinuous_0, ModulatorContinuous_1, ModulatorContinuous_2], properties {inputVolume [0.75,1.2,1.2]}}
 {id=ModulatorContinuous_0, children [WavePlayer_1], properties {volumeModulationParams {default=0, parameterName=Unaware}}}
 {id=ModulatorContinuous_1, children [WavePlayer_0], properties {volumeModulationParams {default=0, parameterName=Alerted}}}
 {id=ModulatorContinuous_2, children [WavePlayer_0], properties {volumeModulationParams {default=0, parameterName=Threaten}}}
 {id=WavePlayer_0, properties {bLooping=true, soundWaveAssetPtr=&/Mx_VampireBoss_Aggro_Loop}}
 {id=WavePlayer_1, properties {bLooping=true, soundWaveAssetPtr=&/Mx_VampireBoss_Idle_Loop}}
]
properties {
 maxDistance=2097152
 soundClassObject=&Fort_Music_NPC
 volumeMultiplier=1.2
 soundSubmixSends [{minSendLevel=0, soundSubmix=&Fortnite_AudioAnalysis_Music}]
}
